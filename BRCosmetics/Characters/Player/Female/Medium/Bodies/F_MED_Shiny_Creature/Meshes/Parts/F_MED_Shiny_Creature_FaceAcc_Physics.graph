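# Physics asset for the Shiny_Creature female medium face accessory, setting up physics for its left and right bangs.
# Capsule bodies form a four-bone left chain and a five-bone right chain hanging from the head, alongside head, spine_05 and unconstrained upper-arm bodies.
# Tight swing and near-zero twist limits keep the bangs stiff, with right_top to right_01 almost rigid.
# The last right bone is spelled dyn_bnags_right_04.
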PhysicsAsset /BRCosmetics/Characters/Player/Female/Medium/Bodies/F_MED_Shiny_Creature/Meshes/Parts/F_MED_Shiny_Creature_FaceAcc_Physics
bodies {
 dyn_bangs_left_01 [{
   shapeType=Capsule
   length=3.0746
   radius=1.2251
   shapeName=dyn_bangs_left_01_capsule
   center=<1.0551 0.1354 -0.1642>
   rotation=<rot 1.2229 88.0219 -82.5574>
  }]
 dyn_bangs_left_02 [{
   shapeType=Capsule
   length=4.0704
   radius=1.2251
   shapeName=dyn_bangs_left_02_capsule
   center=<0.9093 0.0321 0.0639>
   rotation=<rot 0 -87.6427 -56.9064>
  }]
 dyn_bangs_left_03 [{
   shapeType=Capsule
   length=3.4891
   radius=0.967
   shapeName=dyn_bangs_left_03_capsule
   center=<0.8501 -0.2581 -0.3956>
   rotation=<rot 0 -91.8718 -98.1981>
  }]
 dyn_bangs_left_top [{
   shapeType=Capsule
   length=3.5707
   radius=1.8389
   shapeName=dyn_bangs_left_top_capsule
   center=<-0.6023 -1.1101 -2.9783>
   rotation=<rot 0.0009 6.8118 -171.2716>
  }]
 dyn_bangs_right_01 [{
   shapeType=Capsule
   length=4.0528
   radius=1.7852
   shapeName=dyn_bangs_right_01_capsule
   center=<-0.6505 0.8483 0.7579>
   rotation=<rot 0 -95.3313 -87.5236>
  }]
 dyn_bangs_right_02 [{
   shapeType=Capsule
   length=3.4434
   radius=1.5203
   shapeName=dyn_bangs_right_02_capsule
   center=<1.0989 0.0914 -0.3296>
   rotation=<rot 0 100.3427 -65.0666>
  }]
 dyn_bangs_right_03 [{
   shapeType=Capsule
   length=1.9647
   radius=1.2877
   shapeName=dyn_bangs_right_03_capsule
   center=<1.3713 0.0989 -0.0645>
   rotation=<rot 0 83.8903 -89.866>
  }]
 dyn_bangs_right_top [{
   shapeType=Capsule
   length=4.1908
   radius=1.7372
   shapeName=dyn_bangs_right_top_capsule
   center=<0.4072 -1.74 -1.3599>
   rotation=<rot 0.001 171.9942 160.333>
  }]
 dyn_bnags_right_04 [{
   shapeType=Capsule
   length=2.1409
   radius=0.8296
   shapeName=dyn_bnags_right_04_capsule
   center=<0.2994 0.3193 0.1319>
   rotation=<rot 0 -71.4169 -95.3537>
  }]
 head [{shapeType=Capsule, length=6.112, radius=6.0083, center=<2.1956 0.3666 -0.0873>, rotation=<rot 2.5696 -82.1906 -89.9074>}]
 spine_05 [
  {shapeType=Capsule, length=14.9715, radius=11.8138, center=<1.7739 -1.8766 -0.0027>, rotation=<rot 0 -162.4588 179.8729>}
  {
   shapeType=Capsule
   length=24.6476
   radius=5.3524
   shapeName=spine_05_capsule
   center=<16.6779 0.2826 -14.9142>
   rotation=<rot 0.0263 5.6459 -90.2657>
  }
 ]
 upperarm_l [{
   shapeType=Capsule
   length=27.25
   radius=2.75
   shapeName=upperarm_l_capsule
   center=<10.9626 0 -0.2487>
   rotation=<rot 88.7006 -0.0007 -0.0007>
  }]
 upperarm_r [{
   shapeType=Capsule
   length=25.75
   radius=4.25
   shapeName=upperarm_r_capsule
   center=<-12.2709 0 0.2783>
   rotation=<rot 88.7006 -0.0007 -0.0007>
  }]
}
constraints [
 {bone1Name=head, bone2Name=spine_05, swing1LimitDegrees=45, swing2LimitDegrees=45, twistLimitDegrees=45}
 {bone1Name=dyn_bangs_right_top, bone2Name=head, swing1LimitDegrees=15, swing2LimitDegrees=15, twistLimitDegrees=1}
 {bone1Name=dyn_bangs_right_01, bone2Name=dyn_bangs_right_top, swing1LimitDegrees=1, swing2LimitDegrees=1, twistLimitDegrees=0.1}
 {bone1Name=dyn_bangs_right_02, bone2Name=dyn_bangs_right_01, swing1LimitDegrees=15, swing2LimitDegrees=15, twistLimitDegrees=1}
 {bone1Name=dyn_bangs_right_03, bone2Name=dyn_bangs_right_02, swing1LimitDegrees=15, swing2LimitDegrees=15, twistLimitDegrees=1}
 {bone1Name=dyn_bnags_right_04, bone2Name=dyn_bangs_right_03, swing1LimitDegrees=15, swing2LimitDegrees=15, twistLimitDegrees=1}
 {bone1Name=dyn_bangs_left_top, bone2Name=head, swing1LimitDegrees=15, swing2LimitDegrees=15, twistLimitDegrees=1}
 {bone1Name=dyn_bangs_left_03, bone2Name=dyn_bangs_left_02, swing1LimitDegrees=15, swing2LimitDegrees=15, twistLimitDegrees=0.1}
 {bone1Name=dyn_bangs_left_02, bone2Name=dyn_bangs_left_01, swing1LimitDegrees=15, swing2LimitDegrees=15, twistLimitDegrees=0.1}
 {bone1Name=dyn_bangs_left_01, bone2Name=dyn_bangs_left_top, swing1LimitDegrees=15, swing2LimitDegrees=15, twistLimitDegrees=0.1}
]
