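# Post-process-style secondary-motion AnimBP for the Shiny Creature hood-up face accessory, animating hat, bangs and earring bones.
# It copies the parent mesh pose, then layers additive bone offsets, AnimDynamics and a neck-based RigidBody simulation.
# EventGraph derives offset alphas from shoulder rotation and enables a hat offset for assault, forward-grip and sniper rifles.
# Bangs constraints are disconnected, top-bangs dynamics have zero alpha, and shoulder-up values are unused.
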AnimBlueprint /BRCosmetics/Characters/Player/Female/Medium/Bodies/F_MED_Shiny_Creature/Meshes/Parts/F_MED_Shiny_Creature_Hood_Up_FaceAcc_AnimBP
target_skeleton=&/F_MED_Shiny_Creature_Hood_Up_FaceAcc_Skeleton
graphs [
 AnimGraph {nodes [
   n1 = "Constraint - dyn_bangs_left_02 (PP)"!(ComponentPose, Alpha=1, ConstraintWeights_0=0.3, ConstraintWeights_1:float=0.7) -> (Pose) {
    alphaBoolBlend {blendOption=Linear}
    boneToModify {boneName=dyn_bangs_left_02}
    constraintSetup [{targetBone {boneName=spine_05}}, {targetBone {boneName=dyn_bangs_left_top}}]
    constraintWeights [1,1]
   }
   "Constraint - dyn_bangs_right_02 (PP)"!(ComponentPose=@n1, Alpha=1, ConstraintWeights_0=0.2, ConstraintWeights_1:float=0.8) -> (Pose) {
    alphaBoolBlend {blendOption=Linear}
    boneToModify {boneName=dyn_bangs_right_02}
    constraintSetup [{targetBone {boneName=spine_05}}, {targetBone {boneName=dyn_bangs_right_top}}]
    constraintWeights [1,1]
   }
   n2 = "Transform (Modify) Bone - Bone: dyn_earing_left"(
    ComponentPose=Local To Component(LocalPose=Copy Pose From Mesh() {bUseAttachedParent=true})
    Alpha=1
    Translation="0,0,0"
    Rotation="0,0,-15"
    Scale="1,1,1"
   ) {
    rotationMode=BMM_Additive
    rotationSpace=BCS_BoneSpace
    scaleMode=BMM_Additive
    scaleSpace=BCS_BoneSpace
    translationMode=BMM_Additive
    translationSpace=BCS_BoneSpace
    alphaBoolBlend {blendOption=Linear}
    boneToModify {boneName=dyn_earing_left}
    rotation {roll=-15}
   }
   n3 = "Transform (Modify) Bone - Bone: dyn_hat_base_l"(ComponentPose=@n2, Alpha=Get Shoulder_rot_f_l() -> (Shoulder_rot_f_l:double), Translation="-15,0,0", Rotation="40,0,0", Scale="1,1,1") {
    rotationMode=BMM_Additive
    rotationSpace=BCS_BoneSpace
    scaleMode=BMM_Additive
    scaleSpace=BCS_BoneSpace
    translationMode=BMM_Additive
    translationSpace=BCS_BoneSpace
    alphaBoolBlend {blendOption=Linear}
    boneToModify {boneName=dyn_hat_base_l}
    rotation {pitch=40}
    translation {x=-15}
   }
   n4 = "Transform (Modify) Bone - Bone: dyn_hat_base_r"(ComponentPose=@n3, Alpha=Get Shoulder_rot_f_r() -> (Shoulder_rot_f_r:double), Translation="15,0,0", Rotation="20,0,0", Scale="1,1,1") {
    rotationMode=BMM_Additive
    rotationSpace=BCS_BoneSpace
    scaleMode=BMM_Additive
    scaleSpace=BCS_BoneSpace
    translationMode=BMM_Additive
    translationSpace=BCS_BoneSpace
    alphaBoolBlend {blendOption=Linear}
    boneToModify {boneName=dyn_hat_base_r}
    rotation {pitch=20}
    translation {x=15}
   }
   n5 = "Transform (Modify) Bone - Bone: dyn_hat_r_gun_r"(ComponentPose=@n4, Alpha=Get offset_weapon() -> (offset_weapon:double), Translation="2,-5,0", Rotation="-10,0,-30", Scale="1,1,1") {
    rotationMode=BMM_Additive
    rotationSpace=BCS_BoneSpace
    scaleMode=BMM_Additive
    scaleSpace=BCS_BoneSpace
    translationMode=BMM_Additive
    translationSpace=BCS_BoneSpace
    alphaBoolBlend {blendOption=Linear}
    boneToModify {boneName=dyn_hat_r_gun_r}
    rotation {pitch=-10, roll=-30}
    translation=<2 -5>
   }
   n6 = Anim Dynamics - Bone: dyn_earing_right(ComponentPose=@n5, Alpha=1, ExternalForce="0,0,0") {
    angularDampingOverride=0.9
    angularSpringConstant=1000
    bAngularSpring=true
    bGravityOverrideInSimSpace=true
    alphaBoolBlend {blendOption=Linear}
    boundBone {boneName=dyn_earing_right}
    componentLinearAccScale=<1.3 1.3 1.3>
    componentLinearVelScale=<1 1 1>
    physicsBodyDefinitions [{
      sphereCollisionRadius=0
      boundBone {boneName=dyn_earing_right}
      boxExtents=<20 3 20>
      constraintSetup {twistAxis=AxisY, angularLimitsMax=<0 0 5>, angularLimitsMin=<0 0 -10>, angularTarget=<0 0 0>}
      localJointOffset=<-1 0 0>
     }]
    relativeSpaceBone {boneName=pelvis}
   }
   n7 = Anim Dynamics - Bone: dyn_earing_left(ComponentPose=@n6, Alpha=1, ExternalForce="0,0,0") {
    angularDampingOverride=0.9
    angularSpringConstant=1000
    bAngularSpring=true
    bGravityOverrideInSimSpace=true
    alphaBoolBlend {blendOption=Linear}
    boundBone {boneName=dyn_earing_left}
    componentLinearAccScale=<1.3 1.3 1.3>
    componentLinearVelScale=<1 1 1>
    physicsBodyDefinitions [{
      sphereCollisionRadius=0
      boundBone {boneName=dyn_earing_left}
      boxExtents=<20 3 20>
      constraintSetup {twistAxis=AxisY, angularLimitsMax=<0 0 5>, angularLimitsMin=<0 0 -10>, angularTarget=<0 0 0>}
      localJointOffset=<-1 0 0>
     }]
    relativeSpaceBone {boneName=pelvis}
   }
   n8 = Anim Dynamics - Bone: dyn_bangs_left_top(ComponentPose=@n7, Alpha=0, ExternalForce="0,0,0") {
    angularDampingOverride=0.999
    bAngularSpring=true
    bGravityOverrideInSimSpace=true
    bOverrideAngularDamping=true
    bUseGravityOverride=true
    simulationSpace=RootRelative
    alphaBoolBlend {blendOption=Linear}
    boundBone {boneName=dyn_bangs_left_top}
    componentAppliedLinearAccClamp=<2500 2500 2500>
    componentLinearAccScale=<0.3 0.3 0.35>
    componentLinearVelScale=<0.35 0.35 0.35>
    gravityOverride {z=-600}
    physicsBodyDefinitions [{
      sphereCollisionRadius=0
      boundBone {boneName=dyn_bangs_left_top}
      boxExtents=<10 10 5>
      constraintSetup {angularTargetAxis=AxisY, twistAxis=AxisY, angularLimitsMax=<0 5 0>, angularLimitsMin=<-10 -2 -25>, angularTarget=<0 0 0>}
      localJointOffset=<0 -0.1 -3>
     }]
    relativeSpaceBone {boneName=spine_05}
   }
   n9 = Anim Dynamics - Bone: dyn_bangs_left_01(ComponentPose=@n8, Alpha=0.5, ExternalForce="0,0,0") {
    angularDampingOverride=0.999
    bAngularSpring=true
    bGravityOverrideInSimSpace=true
    bOverrideAngularDamping=true
    bUseGravityOverride=true
    simulationSpace=RootRelative
    alphaBoolBlend {blendOption=Linear}
    boundBone {boneName=dyn_bangs_left_01}
    componentAppliedLinearAccClamp=<2500 2500 2500>
    componentLinearAccScale=<0.3 0.3 0.35>
    componentLinearVelScale=<0.35 0.35 0.35>
    gravityOverride {z=-600}
    physicsBodyDefinitions [{
      sphereCollisionRadius=0
      boundBone {boneName=dyn_bangs_left_01}
      boxExtents=<5 10 10>
      constraintSetup {angularTargetAxis=AxisY, angularLimitsMax=<0 5 10>, angularLimitsMin=<-10 -2 -10>, angularTarget=<0 0 0>}
      localJointOffset=<3 -0.1 0>
     }]
    relativeSpaceBone {boneName=spine_05}
   }
   n10 = Anim Dynamics - Bone: dyn_bangs_left_02(ComponentPose=@n9, Alpha=0.5, ExternalForce="0,0,0") {
    angularDampingOverride=0.999
    bAngularSpring=true
    bGravityOverrideInSimSpace=true
    bOverrideAngularDamping=true
    bUseGravityOverride=true
    simulationSpace=RootRelative
    alphaBoolBlend {blendOption=Linear}
    boundBone {boneName=dyn_bangs_left_02}
    componentAppliedLinearAccClamp=<2500 2500 2500>
    componentLinearAccScale=<0.3 0.3 0.35>
    componentLinearVelScale=<0.35 0.35 0.35>
    gravityOverride {z=-600}
    physicsBodyDefinitions [{
      sphereCollisionRadius=0
      boundBone {boneName=dyn_bangs_left_02}
      boxExtents=<5 10 10>
      constraintSetup {angularTargetAxis=AxisY, angularLimitsMax=<0 5 10>, angularLimitsMin=<-10 -2 -10>, angularTarget=<0 0 0>}
      localJointOffset=<0 -0.1 -3>
     }]
    relativeSpaceBone {boneName=spine_05}
   }
   n11 = Anim Dynamics - Bone: dyn_bangs_left_03(ComponentPose=@n10, Alpha=0.5, ExternalForce="0,0,0") {
    angularDampingOverride=0.999
    bAngularSpring=true
    bGravityOverrideInSimSpace=true
    bOverrideAngularDamping=true
    bUseGravityOverride=true
    simulationSpace=RootRelative
    alphaBoolBlend {blendOption=Linear}
    boundBone {boneName=dyn_bangs_left_03}
    componentAppliedLinearAccClamp=<2500 2500 2500>
    componentLinearAccScale=<0.3 0.3 0.35>
    componentLinearVelScale=<0.35 0.35 0.35>
    gravityOverride {z=-600}
    physicsBodyDefinitions [{
      sphereCollisionRadius=0
      boundBone {boneName=dyn_bangs_left_03}
      boxExtents=<5 10 10>
      constraintSetup {angularTargetAxis=AxisY, angularLimitsMax=<0 5 25>, angularLimitsMin=<-2 -2 0>, angularTarget=<0 0 0>}
      localJointOffset=<0 -0.1 -3>
     }]
    relativeSpaceBone {boneName=spine_05}
   }
   n12 = Anim Dynamics - Bone: dyn_bangs_right_top(ComponentPose=@n11, Alpha=0, ExternalForce="0,0,0") {
    angularDampingOverride=0.999
    bAngularSpring=true
    bGravityOverrideInSimSpace=true
    bOverrideAngularDamping=true
    bUseGravityOverride=true
    simulationSpace=RootRelative
    alphaBoolBlend {blendOption=Linear}
    boundBone {boneName=dyn_bangs_right_top}
    componentAppliedLinearAccClamp=<2500 2500 2500>
    componentLinearAccScale=<0.3 0.3 0.35>
    componentLinearVelScale=<0.35 0.35 0.35>
    gravityOverride {z=-600}
    physicsBodyDefinitions [{
      sphereCollisionRadius=0
      boundBone {boneName=dyn_bangs_right_top}
      boxExtents=<5 10 10>
      constraintSetup {angularTargetAxis=AxisY, angularLimitsMax=<5 5 5>, angularLimitsMin=<-2 -5 -5>, angularTarget=<0 0 0>}
      localJointOffset=<5 -0.1 -1>
     }]
    relativeSpaceBone {boneName=spine_05}
   }
   n13 = Anim Dynamics - Bone: dyn_bangs_right_01(ComponentPose=@n12, Alpha=0.5, ExternalForce="0,0,0") {
    angularDampingOverride=0.999
    bAngularSpring=true
    bGravityOverrideInSimSpace=true
    bOverrideAngularDamping=true
    bUseGravityOverride=true
    simulationSpace=RootRelative
    alphaBoolBlend {blendOption=Linear}
    boundBone {boneName=dyn_bangs_right_01}
    componentAppliedLinearAccClamp=<2500 2500 2500>
    componentLinearAccScale=<0.3 0.3 0.35>
    componentLinearVelScale=<0.35 0.35 0.35>
    gravityOverride {z=-600}
    physicsBodyDefinitions [{
      sphereCollisionRadius=0
      boundBone {boneName=dyn_bangs_right_01}
      boxExtents=<5 10 10>
      constraintSetup {angularTargetAxis=AxisY, angularLimitsMax=<5 5 5>, angularLimitsMin=<-2 -5 -5>, angularTarget=<0 0 0>}
      localJointOffset=<5 -0.1 -1>
     }]
    relativeSpaceBone {boneName=spine_05}
   }
   n14 = Anim Dynamics - Bone: dyn_bangs_right_02(ComponentPose=@n13, Alpha=0.5, ExternalForce="0,0,0") {
    angularDampingOverride=0.999
    bAngularSpring=true
    bGravityOverrideInSimSpace=true
    bOverrideAngularDamping=true
    bUseGravityOverride=true
    simulationSpace=RootRelative
    alphaBoolBlend {blendOption=Linear}
    boundBone {boneName=dyn_bangs_right_02}
    componentAppliedLinearAccClamp=<2500 2500 2500>
    componentLinearAccScale=<0.3 0.3 0.35>
    componentLinearVelScale=<0.35 0.35 0.35>
    gravityOverride {z=-600}
    physicsBodyDefinitions [{
      sphereCollisionRadius=0
      boundBone {boneName=dyn_bangs_right_02}
      boxExtents=<5 10 10>
      constraintSetup {angularTargetAxis=AxisY, angularLimitsMax=<5 5 5>, angularLimitsMin=<-2 -5 -5>, angularTarget=<0 0 0>}
      localJointOffset=<5 -0.1 -1>
     }]
    relativeSpaceBone {boneName=spine_05}
   }
   n15 = Anim Dynamics - Bone: dyn_bangs_right_03(ComponentPose=@n14, Alpha=0.5, ExternalForce="0,0,0") {
    angularDampingOverride=0.999
    bAngularSpring=true
    bGravityOverrideInSimSpace=true
    bOverrideAngularDamping=true
    bUseGravityOverride=true
    simulationSpace=RootRelative
    alphaBoolBlend {blendOption=Linear}
    boundBone {boneName=dyn_bangs_right_03}
    componentAppliedLinearAccClamp=<2500 2500 2500>
    componentLinearAccScale=<0.3 0.3 0.35>
    componentLinearVelScale=<0.35 0.35 0.35>
    gravityOverride {z=-600}
    physicsBodyDefinitions [{
      sphereCollisionRadius=0
      boundBone {boneName=dyn_bangs_right_03}
      boxExtents=<5 10 10>
      constraintSetup {angularTargetAxis=AxisY, angularLimitsMax=<5 5 5>, angularLimitsMin=<-2 -5 -5>, angularTarget=<0 0 0>}
      localJointOffset=<5 -0.1 -1>
     }]
    relativeSpaceBone {boneName=spine_05}
   }
   n16 = Anim Dynamics - Bone: dyn_bnags_right_04(ComponentPose=@n15, Alpha=0.5, ExternalForce="0,0,0") {
    angularDampingOverride=0.999
    bAngularSpring=true
    bGravityOverrideInSimSpace=true
    bOverrideAngularDamping=true
    bUseGravityOverride=true
    simulationSpace=RootRelative
    alphaBoolBlend {blendOption=Linear}
    boundBone {boneName=dyn_bnags_right_04}
    componentAppliedLinearAccClamp=<2500 2500 2500>
    componentLinearAccScale=<0.3 0.3 0.35>
    componentLinearVelScale=<0.35 0.35 0.35>
    gravityOverride {z=-600}
    physicsBodyDefinitions [{
      sphereCollisionRadius=0
      boundBone {boneName=dyn_bnags_right_04}
      boxExtents=<5 10 10>
      constraintSetup {angularTargetAxis=AxisY, angularLimitsMax=<5 5 5>, angularLimitsMin=<-2 -5 -5>, angularTarget=<0 0 0>}
      localJointOffset=<5 -0.1 -1>
     }]
    relativeSpaceBone {boneName=spine_05}
   }
   n17 = "Transform (Modify) Bone - Bone: dyn_bangs_right_01"(ComponentPose=@n16, Alpha=Get Shoulder_rot_f_r(), Translation="0,0,0", Rotation="20,0,0", Scale="1,1,1") {
    rotationMode=BMM_Additive
    rotationSpace=BCS_BoneSpace
    scaleMode=BMM_Additive
    scaleSpace=BCS_BoneSpace
    translationMode=BMM_Additive
    translationSpace=BCS_BoneSpace
    alphaBoolBlend {blendOption=Linear}
    boneToModify {boneName=dyn_bangs_right_01}
    rotation {pitch=20}
   }
   n18 = "Transform (Modify) Bone - Bone: dyn_bangs_left_01"(ComponentPose=@n17, Alpha=Get Shoulder_rot_f_l(), Translation="0,0,0", Rotation="-15,0,0", Scale="1,1,1") {
    rotationMode=BMM_Additive
    rotationSpace=BCS_BoneSpace
    scaleMode=BMM_Additive
    scaleSpace=BCS_BoneSpace
    translationMode=BMM_Additive
    translationSpace=BCS_BoneSpace
    alphaBoolBlend {blendOption=Linear}
    boneToModify {boneName=dyn_bangs_left_01}
    rotation {pitch=-15}
   }
   n19 = RigidBody!(ComponentPose=@n18, Alpha, OverrideWorldGravity:Vector="0,0,-980", ExternalForce) -> (Pose) {
    bClampLinearTranslationLimitToRefPose=true
    bOverrideWorldGravity=true
    overridePhysicsAsset=&/F_MED_Shiny_Creature_Hood_Up_FaceAcc_Physics1
    simulationSpace=BaseBoneSpace
    alphaBoolBlend {blendOption=Linear}
    baseBoneRef {boneName=neck_01}
    componentAppliedLinearAccClamp=<1200 1200 1200>
    componentLinearAccScale=<0.2 0.2 0.25>
    componentLinearVelScale=<0.25 0.25 0.25>
    overrideWorldGravity {z=-980}
    simSpaceSettings {
     maxAngularAcceleration=1000
     maxAngularVelocity=0.3
     maxLinearAcceleration=700
     maxLinearVelocity=500
     velocityScaleZ=0
     worldAlpha=0.1
     externalLinearDragV=<0.5 0.5>
    }
   }
   Output Pose(Result=Component To Local(ComponentPose=@n19)) {layerGroup=None}
  ]}
 EventGraph {nodes [
   Event BlueprintUpdateAnimation()
   n1 = Get Shoulder_Mesh() -> (Shoulder_Mesh:Skeletal Mesh Component Object Reference)
   > Set Shoulder_rot_f_r(Shoulder_rot_f_r:double=GetAlphaFromReferenceTransformDelta(SkeletalMeshComponent=@n1, BoneName=upperarm_r, BaseName=clavicle_r, InputRange=(X=0,Y=110), TwistAxis="0,0,1")) -> (Output_Get:double)
   > Set Shoulder_rot_f_l(Shoulder_rot_f_l:double=GetAlphaFromReferenceTransformDelta(SkeletalMeshComponent=@n1, BoneName=upperarm_l, BaseName=clavicle_l, InputRange=(X=0,Y=110), TwistAxis="0,0,1")) -> (Output_Get:double)
   n2 = Get Shoulder_Mesh()
   > Set Shoulder_rot_up_r(Shoulder_rot_up_r:double=GetAlphaFromReferenceTransformDelta(SkeletalMeshComponent=@n2, BoneName=upperarm_r, BaseName=clavicle_r, InputRange=(X=0,Y=30), TwistAxis="0,0,1")) -> (Output_Get:double)
   > Set Shoulder_rot_up_l(Shoulder_rot_up_l:double=GetAlphaFromReferenceTransformDelta(SkeletalMeshComponent=@n2, BoneName=upperarm_l, BaseName=clavicle_l, InputRange=(X=-90,Y=90), TwistAxis="0,0,1")) -> (Output_Get:double)
   > n3 = Cast To FortPlayerAnimInstance(Object=GetAnimInstance(self=GetOwnerSkeletalMeshComponent()))
   > Set offset_weapon(offset_weapon:double=0) -> (Output_Get:double)
   > n4 = Switch on EFortWeaponCoreAnimation!(execute, Selection:EFortWeaponCoreAnimation Enum=Get WeaponCoreAnim(self=@n3)) -> (
    Melee:Exec
    Pistol:Exec
    Shotgun:Exec
    PaperBlueprint:Exec
    Rifle:Exec
    MeleeOneHand:Exec
    MachinePistol:Exec
    RocketLauncher:Exec
    GrenadeLauncher:Exec
    GoingCommando:Exec
    AssaultRifle:Exec
    TacticalShotgun:Exec
    SniperRifle:Exec
    TrapPlacement:Exec
    ShoulderLauncher:Exec
    AbilityDecoTool:Exec
    Crossbow:Exec
    C4:Exec
    RemoteControl:Exec
    DualWield:Exec
    AR_BullPup:Exec
    AR_ForwardGrip:Exec
    MedPackPaddles:Exec
    SMG_P90:Exec
    AR_DrumGun:Exec
    Consumable_Small:Exec
    Consumable_Large:Exec
    Balloon:Exec
    MountedTurret:Exec
    CreativeTool:Exec
    ExplosiveBow:Exec
    AshtonIndigo:Exec
    AshtonChicago:Exec
    MeleeDualWield:Exec
    Unarmed:Exec
   )
   Set offset_weapon(execute=[@n4.AR_ForwardGrip, @n4.AssaultRifle, @n4.SniperRifle], offset_weapon=1)
   Event BlueprintInitializeAnimation()
   > Set Skel_Mesh(Skel_Mesh=GetOwnerSkeletalMeshComponent())
   > n5 = Cast To PlayerPawn_Athena(Object=TryGetPawnOwner())
   > n6 = Set Pawn Owner(Pawn Owner:Fort Player Pawn Athena Object Reference=@n5) -> (Output_Get:Fort Player Pawn Athena Object Reference)
   > Set Shoulder_Mesh(Shoulder_Mesh:Skeletal Mesh Component Object Reference=GetSkeletalMeshForPartType(self=@n6)) -> (Output_Get:Skeletal Mesh Component Object Reference)
  ]}
]
variables [
 offset_weapon {type=number}
 Pawn Owner {type=FortPlayerPawnAthena}
 Shoulder_Mesh {type=SkeletalMeshComponent}
 Shoulder_rot_f_l {type=number}
 Shoulder_rot_f_r {type=number}
 Shoulder_rot_up_l {type=number}
 Shoulder_rot_up_r {type=number}
 Skel_Mesh {type=SkeletalMeshComponent}
]
